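; Creative HVAC vent device that launches actors entering its scalable trigger volume, with force set by a knockup option and falling off with height.
; Options and minigame state set a replicated Enabled flag, whose RepNotify updates trigger collision, wind FX and pitch-shifted audio.
; An authority-side timestamped queue with one rolling timer reapplies the launch to lingering actors.
; Base collision and async-loaded hologram material follow visibility settings, with collision also following play mode.
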(event Custom|Update%20Audio
  (params (On :type Boolean))
  (bind (_audio_on_off_times_x _audio_on_off_times_y) (Math|Vector2D|BreakVector2D (Variables|GetAudio%20On/Off%20Times_X)))
  (bind (_audio_on_off_times_x_2 _audio_on_off_times_y_2) (Math|Vector2D|BreakVector2D (Variables|GetAudio%20On/Off%20Times_X)))
  ;[0.625 0.5564 0] Audio. Spins up or down the audio
  (if On
    (Audio|Components|Audio|FadeIn (Variables|Default|GetLoopingAmbAudio) _audio_on_off_times_x_2 1.0 0.0 "Linear" :!class /Script/Engine.AudioComponent)
    (Components|Timeline|SetPlayRate (Variables|BP_HVAC_Creative|GetAudio%20Pitch%20Shift) (/ 1.0 _audio_on_off_times_x))
    (bind (_direction _pitch_alpha) (Timeline|Audio%20Pitch%20Shift Play)
      (:Update
        (Audio|Components|Audio|SetPitchMultiplier (Variables|Default|GetLoopingAmbAudio) _pitch_alpha))) ; 0s = 0.1\n1s = 1
    (else
      (Audio|Components|Audio|FadeOut (Variables|Default|GetLoopingAmbAudio) _audio_on_off_times_y_2 0.0 "Linear" :!class /Script/Engine.AudioComponent)
      (Components|Timeline|SetPlayRate (Variables|BP_HVAC_Creative|GetAudio%20Pitch%20Shift) (/ 1.0 _audio_on_off_times_y))
      (Timeline|Audio%20Pitch%20Shift Reverse))))

(event EventBeginPlay
  ;[0 1 0.2253] Initialization
  (|CallParentFunction)
  (Class|FortAsyncActionAuthorityMatchReady|MatchReadyAsync
    (:OnReady
      (CallFunction|ResetHVAC)
      (CallFunction|UpdateBaseCollision))) ; Ensures everything exists before getting the initial settings correct. \nWithout this reset HVAC fails initially on the server due to the minigame not existing yet.
)

(event OnAnyPropertyChanged(ToyOptionsComponent)
  ;[0 1 0.2253] Initialization
  (CallFunction|SetForceScale)
  (CallFunction|ScaleTriggerVolume)
  (CallFunction|ResetHVAC)
  (CallFunction|UpdateVisibilityComponent)
  (CallFunction|LoadAndApplyMeshMaterial (not (Variables|UserKnobs|GetbIsVisible)))
  (CallFunction|SetFXScaleAndOffset)
  (CallFunction|UpdateBaseCollision)
  (Components|StaticMesh|SetStaticMesh (Variables|Default|GetSM_HVAC_FX) (select (Variables|Default|GetbEnableVFX) (Variables|Default|GetWindMesh) 0)))

(event OnPropertiesLoaded(ToyOptionsComponent)
  ;[0 1 0.2253] Initialization
  (CallFunction|SetForceScale)
  (CallFunction|ScaleTriggerVolume)
  (CallFunction|ResetHVAC)
  (CallFunction|UpdateVisibilityComponent)
  (CallFunction|LoadAndApplyMeshMaterial (not (Variables|UserKnobs|GetbIsVisible)))
  (CallFunction|SetFXScaleAndOffset)
  (CallFunction|UpdateBaseCollision)
  (Components|StaticMesh|SetStaticMesh (Variables|Default|GetSM_HVAC_FX) (select (Variables|Default|GetbEnableVFX) (Variables|Default|GetWindMesh) 0)))

(event OnAddedtoMinigame(FortMinigameProgress)
  (params (Minigame :type Fort%20Minigame%20Object%20Reference))
  ;[0 1 0.8098 0.25] Play Mode Changing - Update Base Collision
  (Minigame|BindEventtoOnPlayModeChangedDelegate Minigame (delegate Custom|On%20Play%20Mode%20Changed)))

(event MinigameStartedDelegate(FortMinigameProgress)
  ;[0.1719 0.6901 1] Changing Based on Minigame Stage
  (CallFunction|ResetHVAC))

(event Custom|On%20Play%20Mode%20Changed
  (params (Minigame :type Fort%20Minigame%20Object%20Reference) (bIsInPlayMode :type Boolean))
  ;[0 1 0.8098 0.25] Play Mode Changing - Update Base Collision
  (CallFunction|UpdateBaseCollision))

(event MinigameStateChangedDelegate(FortMinigameProgress)
  (params (Minigame :type Fort%20Minigame%20Object%20Reference) (NewMinigameState :type EFortMinigameState%20Enum :default "PreGame"))
  ;[0.1719 0.6901 1] Changing Based on Minigame Stage
  (switch Utilities|FlowControl|Switch|SwitchonEFortMinigameState NewMinigameState
    (:PreGame
      (CallFunction|ResetHVAC))
    (:PostGameEnd
      (CallFunction|LoadAndApplyMeshMaterial (not (Variables|UserKnobs|GetbIsVisible))) ; Called to make sure the correct visibility and material is applied when returning to edit mode
    )))

(event Custom|Load%20and%20Apply%20Mesh%20Material
  (params (Use%20Hologram :type Boolean))
  ;[1 0.6038 0 0.25] Async Load the Appropriate material and apply to the mesh (i.e. if we want to be a hologram)
  (Networking|CanExecuteCosmeticEvents
    (:True
      (bind _loaded_object (Creative|Loading|AsyncLoadEditorSafe (select Use%20Hologram (Variables|Default|GetHologram%20Material) (Variables|Default|GetBase%20Material)))
        (:Loaded
          (bind _asmaterial_interface (Utilities|Casting|CastToMaterialInterface _loaded_object))
          (Rendering|Material|SetMaterial (Variables|Debug|GetStaticMeshComponent) 0 _asmaterial_interface))))))

(event MinigameEndedDelegate(FortMinigameProgress)
  ;[0.1719 0.6901 1] Changing Based on Minigame Stage
  (CallFunction|ResetHVAC))

(event Custom|Enable%20Event
  (params (Instigator :type Controller%20Object%20Reference))
  ;[0.2393 0 1 0.25] Gameplay Functions
  (Variables|Default|SetEnabled true))

(event EventActorEnteredTrigger
  (params (Actor :type Actor%20Object%20Reference) (Hit :type Hit%20Result%20Structure))
  ;[0.1336 1 0.1188] Actor Entered Trigger
  ;;[0.9922 1 0.3191] IF this is a preview building THEN abort launching actors.
  (if (not (Building|IsPreviewBuildingActor))
    (bind _shouldlaunch (Class|CreativeCommonDeviceFunctionLibrary|ApplyVentRecentlyLaunchedTagOrAbortLaunch Actor self))
    (if _shouldlaunch
      (CallFunction|AdjustForceScaleByLocation Hit)
      (|CallParentFunction Actor Hit))))

(event Custom|Disable%20Event
  (params (Instigator :type Controller%20Object%20Reference))
  ;[0.2393 0 1 0.25] Gameplay Functions
  (Variables|Default|SetEnabled false))

(event Custom|Activate%20Event
  (params (Instigator :type Controller%20Object%20Reference))
  ;[0.2393 0 1 0.25] Gameplay Functions
  (if (not (Variables|Default|GetEnabled))
    (Variables|Default|SetEnabled true)
    (Utilities|FlowControl|Delay (Variables|Default|GetActivation%20Time))
    (Variables|Default|SetEnabled false)))

(event Custom|Reapply%20Effect
  (bind _actor_entry_times (Variables|Default|GetActor%20Entry%20Times))
  (bind _output_actor_4_74032141455dd498739795b62d86dc0a (Utilities|Struct|BreakActorTimePair (Utilities|Array|Get(acopy) _actor_entry_times)))
  ;[0 0.432 1] Reapply Effect to actors if they don't leave the volume within a second of entering. Then keep applying.
  (if (Math|Conversions|ToBoolean(Integer) (Utilities|Array|Length (Variables|Default|GetActor%20Entry%20Times)))
    (Utilities|IsValid _output_actor_4_74032141455dd498739795b62d86dc0a
      (:Is%20Valid
        (Utilities|TriggerReapplyEffects _output_actor_4_74032141455dd498739795b62d86dc0a (Transformation|GetActorLocation _output_actor_4_74032141455dd498739795b62d86dc0a)
          (:%20%20
            ;;[1 1 1] Remove from the front, add to the end.
            (Utilities|Array|Add (Variables|Default|GetActor%20Entry%20Times) (Utilities|Struct|MakeActorTimePair _output_actor_4_74032141455dd498739795b62d86dc0a (Utilities|Time|GetGameTimeinSeconds))))))
      (:Is%20Not%20Valid
        _))
    (Utilities|Array|RemoveIndex (Variables|Default|GetActor%20Entry%20Times))
    (CallFunction|StartNextReapplyTimer)))

(event Custom|Start%20Next%20Reapply%20Timer
  (bind (_output_actor_4_74032141455dd498739795b62d86dc0a _output_float_5_25bec5bf409d87f98e7e45a5f7338f8d) (Utilities|Struct|BreakActorTimePair (Utilities|Array|Get(acopy) (Variables|Default|GetActor%20Entry%20Times))))
  ;[0 0.432 1] Reapply Effect to actors if they don't leave the volume within a second of entering. Then keep applying.
  (if (Math|Conversions|ToBoolean(Integer) (Utilities|Array|Length (Variables|Default|GetActor%20Entry%20Times)))
    (bind _returnvalue (Utilities|Time|SetTimerbyEvent (delegate Custom|Reapply%20Effect) (- (Variables|Default|GetReapply%20Time) (- (Utilities|Time|GetGameTimeinSeconds) _output_float_5_25bec5bf409d87f98e7e45a5f7338f8d))))
    (Variables|Default|SetGust%20Reapply%20Timer _returnvalue)))

(event OnComponentBeginOverlap(TriggerVolume)
  (params (OverlappedComponent :type Primitive%20Component%20Object%20Reference) (OtherActor :type Actor%20Object%20Reference) (OtherComp :type Primitive%20Component%20Object%20Reference) (OtherBodyIndex :type Integer) (bFromSweep :type Boolean) (SweepResult :type Hit%20Result%20Structure%20%28by%20ref%29))
  (bind _actor_entry_times (Variables|Default|GetActor%20Entry%20Times))
  ;[1 1 1] Reapply Effect to actors that have not left the volume (common if the knockup is low and the volume size is high)
  (Networking|SwitchHasAuthority
    (:Authority
      (Utilities|Array|Add _actor_entry_times (Utilities|Struct|MakeActorTimePair OtherActor (Utilities|Time|GetGameTimeinSeconds)))
      (if (== (Utilities|Array|LastIndex _actor_entry_times) 0)
        (CallFunction|StartNextReapplyTimer)))))

(event OnComponentEndOverlap(TriggerVolume)
  (params (OverlappedComponent :type Primitive%20Component%20Object%20Reference) (OtherActor :type Actor%20Object%20Reference) (OtherComp :type Primitive%20Component%20Object%20Reference) (OtherBodyIndex :type Integer))
  ;[1 1 1] Reapply Effect to actors that have not left the volume (common if the knockup is low and the volume size is high)
  (CallFunction|RemoveActorFromEntryArray OtherActor))

(fn AdjustForceScalebyLocation
  (params (Hit :type Hit%20Result%20Structure%20%28by%20ref%29))
  (bind (_returnvalue_x _returnvalue_y _returnvalue_z) (Math|Vector|BreakVector (Math|Vector|UnrotateVector (- (.ImpactPoint (Collision|BreakHitResult Hit)) (Transformation|GetWorldLocation (Variables|Default|GetTriggerVolume))) (Transformation|GetWorldRotation (Variables|Default|GetTriggerVolume)))))
  (bind (_returnvalue_x_2 _returnvalue_y_2 _returnvalue_z_2) (Math|Vector|BreakVector (Components|Box|GetScaledBoxExtent (Variables|Default|GetTriggerVolume))))
  (bind (_bblockinghit _binitialoverlap _time _distance _location _impactpoint) (Collision|BreakHitResult Hit))
  (Variables|UserKnobs%28Creative%29|SetDynamicForceScale (* (Variables|Default|GetKnockup%20Scale) (Math|Float|MapRangeClamped _returnvalue_z 5e+01 _returnvalue_z_2 1.0 (Variables|UserKnobs|GetMinKnockbackPercent)))))

(fn AllowInverseScaling
  (return false))

(fn CancelforActorsinVolume
  (for _array_element (Variables|Default|GetActor%20Entry%20Times)
    (bind _array_element_actor_4_74032141455dd498739795b62d86dc0a (Utilities|Struct|BreakActorTimePair _array_element))
    (CallFunction|ActorExitedTrigger _array_element_actor_4_74032141455dd498739795b62d86dc0a))
  (Utilities|Array|Clear (Variables|Default|GetActor%20Entry%20Times))
  (Utilities|Time|ClearandInvalidateTimerbyHandle (Variables|Default|GetGust%20Reapply%20Timer)))

(fn ForceGridAlignment
  (return false))

(fn GetAdditionalSelectableActors
  (return 0))

(fn GetBlockPitchAndRoll
  (return false))

(fn GetScaleMax
  (params (DefaultScaleMax :type Float%20%28single-precision%29))
  (return (Variables|Default|GetNewMaxScale)))

(fn GetScaleMin
  (params (DefaultScaleMin :type Float%20%28single-precision%29))
  (return DefaultScaleMin))

(fn GetSpawnedActors
  (return 0))

(fn HasAdditionalSelectableActors
  (return false))

(fn HasAdditionalValidPlacementLogic
  (return false))

(fn IsTargetableByLinearMover
  (return true))

(fn IsValidPlacement
  (return false))

(fn OnRetrySetCollision
  (params (Minigame :type Fort%20Minigame%20Object%20Reference))
  ;[0 1 0.1847 0.316] If Update Base Collision is called without enough information (Minigame not yet initialized). We'll bind this function to the 'On Added to Minigame' delegate to retry once ready.
  (CallFunction|UpdateBaseCollision))

(fn OnPreviewEnd
  (return false))

(fn OnPreviewStart
  (return false))

(fn OnRep_Enabled
  (CallFunction|UpdateHVACValues))

(fn RemoveActorfromEntryArray
  (params (Actor :type Actor%20Object%20Reference))
  ;[1 0 0] Not happy with my implementation here.
  ; Is this slower than having everything in an actor/time map and looking for lowest float each time we want to readd?
  (LocalVariables|Setlocal%20Actor%20to%20Find Actor)
  (bind (_array_element _array_index) (Utilities|Array|ForEachLoopwithBreak (Variables|Default|GetActor%20Entry%20Times) :!label _label)
    (:LoopBody
      (bind _array_element_actor_4_74032141455dd498739795b62d86dc0a (Utilities|Struct|BreakActorTimePair _array_element))
      (if (Utilities|Equal(Object) _array_element_actor_4_74032141455dd498739795b62d86dc0a (LocalVariables|Getlocal%20Actor%20to%20Find))
        (Utilities|Array|RemoveIndex (Variables|Default|GetActor%20Entry%20Times) _array_index)
        (goto _label Break)))))

(fn ResetHVAC
  (bind _fortminigameprogress (Variables|Default|GetFortMinigameProgress)
    (:then
      (bind _currentminigame (Variables|/Script/FortniteGame.FortMinigameLogicComponent|GetCurrentMinigame _fortminigameprogress)
        (:then
          (LocalVariables|Setlocal%20Desired%20Enabled (Math|Boolean|ORBoolean (and (== (Variables|UserKnobs|GetEnabledIndex) 1) (Utilities|Enum|Equal(Enum) (Variables|/Script/FortniteGame.FortMinigame|GetCurrentState _currentminigame) "PreGame")) (and (Utilities|Enum|Equal(Enum) (Variables|/Script/FortniteGame.FortMinigame|GetCurrentState _currentminigame) "InProgress") (== (Variables|UserKnobs|GetEnabledIndex) 2)) (== (Variables|UserKnobs|GetEnabledIndex) 0))))
        (:else
          (LocalVariables|Setlocal%20Desired%20Enabled (or (== (Variables|UserKnobs|GetEnabledIndex) 0) (== (Variables|UserKnobs|GetEnabledIndex) 1)) :!label _label)))
      (if (Math|Boolean|NotEqual(Boolean) (LocalVariables|Getlocal%20Desired%20Enabled) (Variables|Default|GetEnabled))
        (Variables|Default|SetEnabled (LocalVariables|Getlocal%20Desired%20Enabled))
        (else
          (CallFunction|UpdateHVACValues) ; If the RepNotify wouldn't trigger, manually call the function that it instigates.
        )))
    (:else
      _))
  (goto _label))

(fn ScaleTriggerVolume
  (bind _triggerrange (Variables|UserKnobs|GetTriggerRange))
  (bind _returnvalue (* 5e+01 _triggerrange))
  (bind _base_height (Variables|Default|GetBase%20Height))
  (Utilities|FlowControl|Sequence
    (:then_0
      (LocalVariables|Setlocal%20ExtentZ (- _returnvalue _base_height))
      (LocalVariables|Setlocal%20Offset (+ _base_height _returnvalue)))
    (:then_1
      ;[1 1 1] Offset trigger volume (or use default values from parent)
      (Transformation|SetRelativeLocation (Variables|Default|GetTriggerVolume) (Math|Vector|MakeVector 0.0 0.0 (LocalVariables|Getlocal%20Offset)) false false)
      (Components|Box|SetBoxExtent (Variables|Default|GetTriggerVolume) (Math|Vector|MakeVector 1.4e+02 1.4e+02 (LocalVariables|Getlocal%20ExtentZ)) true))))

(fn ServerHVACToggle
  (params (On :type Boolean))
  (Networking|SwitchHasAuthority
    (:Authority
      (Variables|Default|SetEnabled On))))

(fn SetForceScale
  (if (!= (Variables|UserKnobs|GetKnockupIndex) -1)
    (Variables|Default|SetKnockup%20Scale (Utilities|Array|Get(acopy) (Variables|UserKnobs|GetKnockupMultipliers) (Variables|UserKnobs|GetKnockupIndex)))
    (else
      (Variables|Default|SetKnockup%20Scale 1.0))))

(fn SetFXScaleandOffset
  (Networking|CanExecuteCosmeticEvents
    (:True
      (Material|ReattachOrCreateMID (Variables|Default|GetSM_HVAC_FX) (Variables|Default|GetFX_MID))
      (LocalVariables|SetIsSpinning (and (> (Variables|UserKnobs|GetKnockupIndex) 0) (Variables|Default|GetbEnableVFX)))
      (Rendering|Material|SetVectorParameterValue (Variables|Default|GetFX_MID) "FanColor" (Math|Color|SelectColor (Variables|Default|GetFanOnColor) "(R=0.000000,G=0.000000,B=0.000000,A=0.000000)" (Math|Boolean|ANDBoolean (Variables|UserKnobs|GetbIsVisible) (Variables|Default|GetEnabled) (LocalVariables|GetIsSpinning) (Variables|Default|GetbEnableVFX))) :!class /Script/Engine.MaterialInstanceDynamic)
      (bind _output_get (Variables|UserKnobs%28Creative%29|SetFX%20Intensity%20Multiplier (select (LocalVariables|GetIsSpinning) (Math|Float|MapRangeClamped (Math|Conversions|ToFloat(Integer) (Variables|UserKnobs|GetKnockupIndex)) 1.0 6.0 0.65 4.0) 0.0)))
      (Rendering|Material|SetScalarParameterValue (Variables|Default|GetFX_MID) "BP_Intensity" _output_get :!class /Script/Engine.MaterialInstanceDynamic)
      (bind _output_get_2 (Variables|UserKnobs%28Creative%29|SetFX%20Height%20Multiplier (select (LocalVariables|GetIsSpinning) (Math|Float|MapRangeClamped (Variables|UserKnobs|GetTriggerRange) 3.0 15.0 0.75 6.0) 0.0)))
      (Rendering|Material|SetScalarParameterValue (Variables|Default|GetFX_MID) "BP_WPO_MovingUP" _output_get_2 :!class /Script/Engine.MaterialInstanceDynamic)
      (Rendering|SetBoundsScale (Variables|Default|GetSM_HVAC_FX) (Variables|UserKnobs%28Creative%29|GetFX%20Height%20Multiplier)))))

(fn ShouldBlockPlacement
  (params (MoveTool :type Fort%20Creative%20Move%20Tool%20Object%20Reference))
  (return false))

(fn ShouldInvisiblePropsBeAllowed
  (return false))

(fn UpdateBaseCollision
  (bind _bisvisible (Variables|UserKnobs|GetbIsVisible))
  (bind _staticmeshcomponent (Variables|Debug|GetStaticMeshComponent))
  ;[0 1 0.8098 0.25] Updates collision on the base to either No Collision, Collide with everything, or just the interaction channel (for customization)
  (bind _fortminigameprogress (Variables|Default|GetFortMinigameProgress)
    (:then
      (bind _currentminigame (Variables|/Script/FortniteGame.FortMinigameLogicComponent|GetCurrentMinigame _fortminigameprogress)
        (:then
          (Collision|SetCollisionProfileName _staticmeshcomponent (select _bisvisible "FortBuildingMeshPhysics" (select (Minigame|IsinPlayMode _currentminigame :!class /Script/FortniteGame.FortMinigame) "NoCollision" "FortTriggerOnlyInteractions")) false :!class /Script/Engine.PrimitiveComponent)
          (if (not _bisvisible)
            (Collision|SetCollisionResponsetoChannel _staticmeshcomponent "ECC_Visibility" "ECR_Block") ; When the mesh's collision profile is set to "FortTriggerInteractionsOnly" it needs to block visibility in order to select with move tool
          ))
        (:else
          (Minigame|BindEventtoOnAddedtoMinigame _fortminigameprogress (EventDispatchers|CreateEvent :!event On%20Retry%20Set%20Collision)))))
    (:else)))

(fn UpdateHVACValues
  (bind _returnvalue (Utilities|IsEnabled))
  (CallFunction|ScaleTriggerVolume)
  (Collision|SetCollisionEnabled (Variables|Default|GetTriggerVolume) (select (Utilities|IsEnabled) "QueryOnly" "NoCollision"))
  (Utilities|FlowControl|Sequence
    (:then_0
      ;[1 1 1] Updating Client Effects
      (Networking|CanExecuteCosmeticEvents
        (:True
          (Development|SetHiddeninGame (Variables|Default|GetSM_HVAC_FX) (not _returnvalue) false)
          (CallFunction|UpdateAudio (Math|Boolean|ANDBoolean _returnvalue (Variables|UserKnobs|GetbIsVisible) false))
          (CallFunction|SetFXScaleAndOffset))))
    (:then_1
      ;[1 1 1] If we disable the HVAC, cancel for all actors in volume (as they don't auto-trigger the exit volume function)
      (if (not (Utilities|IsEnabled))
        (CallFunction|CancelForActorsInVolume)))))

(fn UpdateVisibilityComponent
  (Visibility|SetVisibilityComponents (Variables|Default|GetVisibleDuringGame) (Utilities|Array|MakeArray (Variables|Debug|GetStaticMeshComponent) (Variables|Default|GetSM_HVAC_FX)) true)
  (Visibility|SetVisiblityDuringPhase (Variables|Default|GetVisibleDuringGame) (select (Variables|UserKnobs|GetbIsVisible) "NewEnumerator3" "NewEnumerator1")))

(fn ConstructionScript
  (|CallParentFunction))
